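# Physics asset of cape colliders for the female medium RedRiding player body, named as the collision set for her cape.
# Capsules cover the head, clavicles, pelvis, spine chain, feet and a tails05 bone, with crossed capsule pairs on the middle spine bones.
# Arm, thigh, calf and upper neck bodies exist but carry no shapes.
# It defines no constraints, so it provides collision volumes rather than a jointed ragdoll.
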PhysicsAsset /Game/Characters/Player/Female/Medium/Bodies/F_MED_RedRiding/Mesh/F_MED_RedRiding_01_CapeColliders
bodies {
 calf_l []
 calf_r []
 clavicle_l [{shapeType=Capsule, length=7.6964, radius=6.3853, center=<4.0147 -0.8354 -1.7419>, rotation=<rot -77.5566 177.3915 7.7314>}]
 clavicle_r [{shapeType=Capsule, length=6.1964, radius=9.7957, center=<-5.6154 0.7093 1.3192>, rotation=<rot -77.5565 177.3915 -172.2685>}]
 foot_l [{shapeType=Capsule, length=12.5982, radius=6.5127, center=<-2.2072 6.0534 -0.3367>, rotation=<rot 0.0388 1.4701 -89.3982>}]
 foot_r [{shapeType=Capsule, length=12.5982, radius=6.5127, center=<2.2072 -6.0534 0.3367>, rotation=<rot 0.0388 1.4701 90.6018>}]
 head [{shapeType=Capsule, length=0, radius=18, center=<9.0697 0.0746 0>, rotation=<rot 0 -90 0>}]
 lowerarm_l []
 lowerarm_r []
 neck_02 []
 pelvis [{shapeType=Capsule, length=16.6528, radius=17.4095, center=<-9.3339 -4.5938 -0.1981>}]
 spine_01 [{shapeType=Capsule, length=7.5171, radius=16.2354, center=<0.8439 -1.0873 0>}]
 spine_02 [
  {shapeType=Capsule, length=3.2984, radius=8.4854, center=<1.179 2.2974 0>}
  {shapeType=Capsule, length=20.6875, radius=9.3125, center=<-1.8317 2.606 0>, rotation=<rot 90 0 0>}
 ]
 spine_03 [
  {shapeType=Capsule, length=3.2984, radius=11.4854, center=<-0.0747 -1.0629 0>}
  {shapeType=Capsule, length=20.7188, radius=9.2812, center=<-1.2326 2.6059 0>, rotation=<rot 90 0 0>}
 ]
 spine_04 [
  {shapeType=Capsule, length=9.5796, radius=10.3916, center=<5.5182 -2.6433 0>}
  {shapeType=Capsule, length=10.0312, radius=9.2812, center=<-1.2326 1.308 0>, rotation=<rot 90 0 0>}
 ]
 spine_05 [{shapeType=Capsule, length=10.9859, radius=10.0791, center=<5.6377 0.3227 0>}]
 tails05 [{shapeType=Capsule, length=13.25, radius=9, shapeName=tails05_capsule, center=<-8.939 0 -0.9862>, rotation=<rot 90 -179.9999 179.9998>}]
 thigh_l []
 thigh_r []
 upperarm_l []
 upperarm_r []
}
constraints []
