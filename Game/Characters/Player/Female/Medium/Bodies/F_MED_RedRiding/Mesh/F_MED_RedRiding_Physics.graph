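# Physics asset for the F_MED_RedRiding female player body, covering a dagger, a bag, tails and the pelvis with right leg, all as capsules.
# The tails chain hangs from the pelvis with limits mostly widening down to a fork at tails03.
# Dagger and upper bag have limited joints to the pelvis, while the lower bag pivots on a small master body.
# That master and the right leg joints rotate freely, and the left leg has no bodies.
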PhysicsAsset /Game/Characters/Player/Female/Medium/Bodies/F_MED_RedRiding/Mesh/F_MED_RedRiding_Physics
bodies {
 bag_lower [{shapeType=Capsule, length=5.7677, radius=2.705, center=<9 0.4274 -0.8443>, rotation=<rot 0 0 64.9997>}]
 bag_lower_MSTR [{shapeType=Capsule, length=0.505, radius=0.505, rotation=<rot 0 -90 0>}]
 bag_upper [{shapeType=Capsule, length=4.292, radius=2.9924, center=<3 0.1066 0.1758>, rotation=<rot 1.1001 -88.0578 -89.9765>}]
 calf_r [{shapeType=Capsule, length=35.532, radius=7.5681, center=<20.9778 2.98 0.2195>, rotation=<rot 16.8715 94.2922 -89.4103>}]
 dagger [{shapeType=Capsule, length=17.5019, radius=5.3992, center=<0 0 -5>, rotation=<rot 0 0 -10.0005>}]
 foot_r [{shapeType=Capsule, length=9.787, radius=8.6795, shapeName=foot_r_capsule, center=<2.2006 -2.1169 1.1058>, rotation=<rot 0 -2.9566 -80.7852>}]
 pelvis [{shapeType=Capsule, length=16.5974, radius=6.6958, center=<0.1359 -0.2876 0.2658>, rotation=<rot -0.3464 -89.5958 -2.2856>}]
 tails01 [{shapeType=Capsule, length=6.4721, radius=3.8507, center=<9.5534 -2.2 -2.5>, rotation=<rot 0 -99.5584 -86.4761>}]
 tails02 [{shapeType=Capsule, length=7.2923, radius=5.1987, center=<6.0833 -4.2 0>, rotation=<rot 0.0001 -89.9993 -97.5144>}]
 tails03 [{shapeType=Capsule, length=4.4751, radius=2.4139, center=<2 0.0955 0.1738>, rotation=<rot -23.5608 89.5251 -92.5571>}]
 tails05 [{shapeType=Capsule, length=2.8845, radius=2.4486, center=<1 0 0>, rotation=<rot 13.6465 89.6925 -91.6948>}]
 tails06 [{shapeType=Capsule, length=7.7957, radius=2.3637, center=<2.5649 0 0>, rotation=<rot -0.0638 -91.7528 -85.3439>}]
 thigh_r [{shapeType=Capsule, length=29.61, radius=7.8427, center=<17.5806 -0.1672 -0.7246>, rotation=<rot 0.0461 -95.2924 -95.4893>}]
}
constraints [
 {bone1Name=dagger, bone2Name=pelvis, swing1LimitDegrees=2, swing2LimitDegrees=2, twistLimitDegrees=20}
 {bone1Name=tails01, bone2Name=pelvis, swing1LimitDegrees=1, swing2LimitDegrees=1, twistLimitDegrees=10}
 {bone1Name=tails02, bone2Name=tails01, swing1LimitDegrees=10, swing2LimitDegrees=10, twistLimitDegrees=5}
 {bone1Name=tails03, bone2Name=tails02, swing1LimitDegrees=15, swing2LimitDegrees=15, twistLimitDegrees=1}
 {
  bone1Name=thigh_r
  bone2Name=pelvis
  swing1LimitDegrees=45
  swing1Motion=Free
  swing2LimitDegrees=45
  swing2Motion=Free
  twistLimitDegrees=45
  twistMotion=Free
 }
 {bone1Name=tails06, bone2Name=tails03, swing1LimitDegrees=28, swing2LimitDegrees=28, twistLimitDegrees=14}
 {bone1Name=tails05, bone2Name=tails03, swing1LimitDegrees=28, swing2LimitDegrees=28, twistLimitDegrees=14}
 {
  bone1Name=calf_r
  bone2Name=thigh_r
  swing1LimitDegrees=45
  swing1Motion=Free
  swing2LimitDegrees=45
  swing2Motion=Free
  twistLimitDegrees=45
  twistMotion=Free
 }
 {bone1Name=bag_upper, bone2Name=pelvis, swing1LimitDegrees=10, swing2LimitDegrees=20, twistLimitDegrees=2}
 {bone1Name=bag_lower, bone2Name=bag_lower_MSTR, swing1LimitDegrees=15, swing2LimitDegrees=45, twistLimitDegrees=2}
 {
  bone1Name=bag_lower_MSTR
  bone2Name=pelvis
  swing1LimitDegrees=45
  swing1Motion=Free
  swing2LimitDegrees=45
  swing2Motion=Free
  twistLimitDegrees=45
  twistMotion=Free
 }
 {
  bone1Name=foot_r
  bone2Name=calf_r
  swing1LimitDegrees=45
  swing1Motion=Free
  swing2LimitDegrees=45
  swing2Motion=Free
  twistLimitDegrees=45
  twistMotion=Free
 }
]
