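# Physics asset for the RedRiding medium female player body, covering secondary motion of pigtails, bangs, hood sides, necklace and shoulder cloth.
# Capsules on upper spine, clavicles and head, linked by free constraints, form the body base; bangs is the only box.
# Mirrored four-segment pigtails and single hood capsules branch from the head through tiny free-jointed base bodies.
# Tight angular limits at pigtail roots widen down each chain; other accessories have their own limits.
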PhysicsAsset /Game/Characters/Player/Female/Medium/Bodies/F_MED_RedRiding/Mesh/F_MED_RedRiding_dyn_hair
bodies {
 bangs [{shapeType=Box, extentX=5.438, extentY=8.176, extentZ=6.0433, center=<1.8173 -0.3622 0.2419>, rotation=<rot 7.8182 -88.0518 -16.1521>}]
 clavicle_l [{shapeType=Capsule, length=12.5876, radius=5.5784, center=<4.7525 -0.1804 -1.2665>, rotation=<rot 33.8122 77.0007 -102.7183>}]
 clavicle_r [{shapeType=Capsule, length=12.5876, radius=5.5784, center=<-3.4154 2.3757 1.1865>, rotation=<rot 35.692 -74.6237 -86.3141>}]
 dyn_hood_l [{
   shapeType=Capsule
   length=4.829
   radius=2.8221
   shapeName=dyn_hood_l_capsule
   center=<4.0657 -2.9875 -1.4335>
   rotation=<rot 0 -90.472 -50.1973>
  }]
 dyn_hood_l_base [{shapeType=Capsule, length=0.505, radius=0.505, shapeName=dyn_hood_l_base_capsule, rotation=<rot 0 -90 0>}]
 dyn_hood_r [{
   shapeType=Capsule
   length=4.829
   radius=2.8221
   shapeName=dyn_hood_r_capsule
   center=<-4.0657 2.9875 1.4335>
   rotation=<rot 0 -90.472 129.8027>
  }]
 dyn_hood_r_base [{shapeType=Capsule, length=0.505, radius=0.505, shapeName=dyn_hood_r_base_capsule, rotation=<rot 0 -90 180>}]
 dyn_pigtail_hood_l_base [{shapeType=Capsule, length=0.505, radius=0.505, shapeName=dyn_pigtail_hood_l_base_capsule, rotation=<rot 0 -90 0>}]
 dyn_pigtail_hood_r_base [{shapeType=Capsule, length=0.505, radius=0.505, shapeName=dyn_pigtail_hood_r_base_capsule, rotation=<rot 0 -90 180>}]
 dyn_pigtail_l_base [{shapeType=Capsule, length=0.505, radius=0.505, shapeName=dyn_pigtail_l_base_capsule, rotation=<rot 0 -90 0>}]
 dyn_pigtail_r_base [{shapeType=Capsule, length=0.505, radius=0.505, shapeName=dyn_pigtail_r_base_capsule, rotation=<rot 0 -90 180>}]
 head [{shapeType=Capsule, length=6.4493, radius=7.032, center=<3.0989 0.3881 -0.3617>, rotation=<rot 6.9979 -90.742 -92.4578>}]
 necklace [{shapeType=Capsule, length=7, radius=2, shapeName=necklace_capsule, center=<7.1552 0 -1.4635>, rotation=<rot 88.9691 179.9998 179.9998>}]
 pigtail_01_l [{shapeType=Capsule, length=5.2265, radius=2.0435, center=<2 0 0>, rotation=<rot 0 -89.2763 -87.8576>}]
 pigtail_01_r [{shapeType=Capsule, length=5.2265, radius=2.0435, center=<-2 0 0>, rotation=<rot 0 -89.2763 92.1424>}]
 pigtail_02_l [{shapeType=Capsule, length=5, radius=2, center=<3 0 0>, rotation=<rot 0 -90 -90>}]
 pigtail_02_r [{shapeType=Capsule, length=5, radius=2, center=<-3 0 0>, rotation=<rot 0 -90 90>}]
 pigtail_03_l [{shapeType=Capsule, length=5, radius=2, center=<3 0 0>, rotation=<rot 0 -90 -90>}]
 pigtail_03_r [{shapeType=Capsule, length=5, radius=2, center=<-3 0 0>, rotation=<rot 0 -90 90>}]
 pigtail_04_l [{shapeType=Capsule, length=5, radius=2, shapeName=pigtail_04_l_capsule, center=<3 0 0>, rotation=<rot 0 -90 -90>}]
 pigtail_04_r [{shapeType=Capsule, length=5, radius=2, shapeName=pigtail_04_r_capsule, center=<-3 0 0>, rotation=<rot 0 -90 90>}]
 shoulder_cloth [{shapeType=Capsule, length=0.384, radius=1.6433, center=<5.8175 -0.2879 -0.8395>, rotation=<rot 68.84 81.813 87.4445>}]
 spine_04 [{shapeType=Capsule, length=2.75, radius=11.0088, center=<-0.0673 1.6661 0.0001>}]
 spine_05 [{shapeType=Capsule, length=7, radius=9.7588, center=<-2.6474 -0.9438 0>}]
}
constraints [
 {
  bone1Name=spine_05
  bone2Name=spine_04
  swing1LimitDegrees=45
  swing1Motion=Free
  swing2LimitDegrees=45
  swing2Motion=Free
  twistLimitDegrees=45
  twistMotion=Free
 }
 {
  bone1Name=clavicle_l
  bone2Name=spine_05
  swing1LimitDegrees=45
  swing1Motion=Free
  swing2LimitDegrees=45
  swing2Motion=Free
  twistLimitDegrees=45
  twistMotion=Free
 }
 {
  bone1Name=clavicle_r
  bone2Name=spine_05
  swing1LimitDegrees=45
  swing1Motion=Free
  swing2LimitDegrees=45
  swing2Motion=Free
  twistLimitDegrees=45
  twistMotion=Free
 }
 {bone1Name=pigtail_01_l, bone2Name=dyn_pigtail_l_base, swing1LimitDegrees=2, swing2LimitDegrees=4, twistLimitDegrees=2}
 {bone1Name=pigtail_02_l, bone2Name=pigtail_01_l, swing1LimitDegrees=8, swing2LimitDegrees=24, twistLimitDegrees=18}
 {bone1Name=pigtail_03_l, bone2Name=pigtail_02_l, swing1LimitDegrees=18, swing2LimitDegrees=24, twistLimitDegrees=18}
 {bone1Name=shoulder_cloth, bone2Name=clavicle_r, swing1LimitDegrees=7, swing2LimitDegrees=7, twistLimitDegrees=1}
 {bone1Name=bangs, bone2Name=head, swing1LimitDegrees=20, swing2LimitDegrees=20, twistLimitDegrees=1}
 {
  bone1Name=head
  bone2Name=spine_05
  swing1LimitDegrees=45
  swing1Motion=Free
  swing2LimitDegrees=45
  swing2Motion=Free
  twistLimitDegrees=45
  twistMotion=Free
 }
 {bone1Name=necklace, bone2Name=spine_05, swing1LimitDegrees=45, swing2LimitDegrees=20, twistLimitDegrees=1}
 {
  bone1Name=dyn_pigtail_hood_l_base
  bone2Name=head
  swing1LimitDegrees=45
  swing1Motion=Free
  swing2LimitDegrees=45
  swing2Motion=Free
  twistLimitDegrees=45
  twistMotion=Free
 }
 {
  bone1Name=dyn_pigtail_l_base
  bone2Name=dyn_pigtail_hood_l_base
  swing1LimitDegrees=45
  swing1Motion=Free
  swing2LimitDegrees=45
  swing2Motion=Free
  twistLimitDegrees=45
  twistMotion=Free
 }
 {bone1Name=pigtail_04_l, bone2Name=pigtail_03_l, swing1LimitDegrees=18, swing2LimitDegrees=24, twistLimitDegrees=18}
 {
  bone1Name=dyn_pigtail_hood_r_base
  bone2Name=head
  swing1LimitDegrees=45
  swing1Motion=Free
  swing2LimitDegrees=45
  swing2Motion=Free
  twistLimitDegrees=45
  twistMotion=Free
 }
 {
  bone1Name=dyn_pigtail_r_base
  bone2Name=dyn_pigtail_hood_r_base
  swing1LimitDegrees=45
  swing1Motion=Free
  swing2LimitDegrees=45
  swing2Motion=Free
  twistLimitDegrees=45
  twistMotion=Free
 }
 {bone1Name=pigtail_01_r, bone2Name=dyn_pigtail_r_base, swing1LimitDegrees=2, swing2LimitDegrees=4, twistLimitDegrees=2}
 {bone1Name=pigtail_02_r, bone2Name=pigtail_01_r, swing1LimitDegrees=8, swing2LimitDegrees=24, twistLimitDegrees=18}
 {bone1Name=pigtail_03_r, bone2Name=pigtail_02_r, swing1LimitDegrees=18, swing2LimitDegrees=24, twistLimitDegrees=18}
 {bone1Name=pigtail_04_r, bone2Name=pigtail_03_r, swing1LimitDegrees=18, swing2LimitDegrees=24, twistLimitDegrees=18}
 {
  bone1Name=dyn_hood_l_base
  bone2Name=dyn_pigtail_hood_l_base
  swing1LimitDegrees=45
  swing1Motion=Free
  swing2LimitDegrees=45
  swing2Motion=Free
  twistLimitDegrees=45
  twistMotion=Free
 }
 {bone1Name=dyn_hood_l, bone2Name=dyn_hood_l_base, swing1LimitDegrees=5, swing2LimitDegrees=5, twistLimitDegrees=5}
 {
  bone1Name=dyn_hood_r_base
  bone2Name=dyn_pigtail_hood_r_base
  swing1LimitDegrees=45
  swing1Motion=Free
  swing2LimitDegrees=45
  swing2Motion=Free
  twistLimitDegrees=45
  twistMotion=Free
 }
 {bone1Name=dyn_hood_r, bone2Name=dyn_hood_r_base, swing1LimitDegrees=5, swing2LimitDegrees=5, twistLimitDegrees=5}
]
