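# Control Rig for the Red Riding medium female player body that procedurally corrects secondary bones for skirt, bag, hood and pigtails.
# Parent and aim constraints attach skirt to thigh twists and blend pigtail and hood bases between head, clavicles and upper spine.
# Spherical pose readers on thighs and head weight corrective offsets, partly through a reusable additive bone-transform function.
# Fixed rotations bend the pigtails, and an extra skirt adjustment runs when ClothEnabled is false.
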ControlRigBlueprint /Game/Characters/Player/Female/Medium/Bodies/F_MED_RedRiding/Mesh/F_MED_Redriding_CtrlRig
graphs [
 ContainedGraph {nodes [
   n1 = RigVMFunctionEntryNode()
   RigUnit_GetTransform = &RigUnit_GetTransform(Space=LocalSpace, bInitial=false, Item=@n1.Joint)
   n2 = &RigVMFunction_MathQuaternionToEuler(RotationOrder=ZYX, Value=@RigUnit_GetTransform.Transform.Rotation)
   > RigUnit_SetTransform = &RigUnit_SetTransform(
    Space=LocalSpace
    Value=(Rotation=(X=0.0132,Y=0.4099,Z=-0.0222,W=0.9118),Translation=(X=23.6045,Y=-6.37,Z=140.1115),Scale3D=(X=1,Y=1,Z=1))
    bInitial=false
    Item=@n1.Joint
    Weight=@n1.Weight
    Value.Translation.Z=&RigVMFunction_MathFloatAdd(B=@n1.Translate Z, A=@RigUnit_GetTransform.Transform.Translation.Z) {template="Add::Execute(in A,in B,out Result)"}.Result
    Value.Scale3D.Z=&RigVMFunction_MathFloatAdd(B=@n1.Scale Z, A=@RigUnit_GetTransform.Transform.Scale3D.Z) {template="Add::Execute(in A,in B,out Result)"}.Result
    Value.Translation.Y=&RigVMFunction_MathFloatAdd(B=@n1.Translate Y, A=@RigUnit_GetTransform.Transform.Translation.Y) {template="Add::Execute(in A,in B,out Result)"}.Result
    Value.Scale3D.Y=&RigVMFunction_MathFloatAdd(B=@n1.Scale Y, A=@RigUnit_GetTransform.Transform.Scale3D.Y) {template="Add::Execute(in A,in B,out Result)"}.Result
    Value.Translation.X=&RigVMFunction_MathFloatAdd(B=@n1.Translate X, A=@RigUnit_GetTransform.Transform.Translation.X) {template="Add::Execute(in A,in B,out Result)"}.Result
    Value.Scale3D.X=&RigVMFunction_MathFloatAdd(B=@n1.Scale X, A=@RigUnit_GetTransform.Transform.Scale3D.X) {template="Add::Execute(in A,in B,out Result)"}.Result
    Value.Rotation=&RigVMFunction_MathQuaternionFromEuler(
     Euler=(X=0,Y=0,Z=0)
     Euler.X=&RigVMFunction_MathFloatAdd(B=@n1.Rotate X, A=@n2.Result.X) {template="Add::Execute(in A,in B,out Result)"}.Result
     Euler.Y=&RigVMFunction_MathFloatAdd(B=@n1.Rotate Y, A=@n2.Result.Y) {template="Add::Execute(in A,in B,out Result)"}.Result
     Euler.Z=&RigVMFunction_MathFloatAdd(B=@n1.Rotate Z, A=@n2.Result.Z) {template="Add::Execute(in A,in B,out Result)"}.Result
    ).Result
   )
   > RigVMFunctionReturnNode()
  ]}
 RigVMFunctionLibrary {nodes [Set Transforms for Bone = RigVMCollapseNode(
    ExecuteContext=()
    Joint=()
    Rotate X=0
    Rotate Y=0
    Rotate Z=0
    Scale X=0
    Scale Y=0
    Scale Z=0
    Translate X=0
    Translate Y=0
    Translate Z=0
    Weight=1
   )]}
 RigVMFunction_Sequence_1_ContainedGraph {nodes [
   n1 = RigVMFunctionEntryNode()
   RigVMFunction_Sequence = &RigVMFunction_Sequence(ExecuteContext=@n1.ExecuteContext)
   n2 = &RigVMFunction_Sequence(ExecuteContext=@RigVMFunction_Sequence.B)
   n3 = &RigVMFunction_Sequence(ExecuteContext=@n2.B)
   n4 = &RigVMFunction_Sequence(ExecuteContext=@n3.B)
   n5 = &RigVMFunction_Sequence(ExecuteContext=@n4.B)
   RigVMFunctionReturnNode(A=@RigVMFunction_Sequence.A, B=@n2.A, C=@n3.A, D=@n4.A, E=@n5.A, F=@n5.B)
  ]}
 RigVMModel {nodes [
   RigUnit_BeginExecution = &RigUnit_BeginExecution()
   RigVMFunction_Sequence = RigVMAggregateNode(ExecuteContext=@RigUnit_BeginExecution.ExecutePin)
   &RigUnit_ParentConstraint(
    AdvancedSettings=(InterpolationType=Average,RotationOrderForFilter=XZY)
    Child=(Type=Bone,Name="skirt_l")
    Parents=((Item=(Type=Bone,Name="thigh_twist_01_l"),Weight=1))
    Weight=1
    ExecutePin=@RigVMFunction_Sequence.A
   )
   > &RigUnit_ParentConstraint(
    AdvancedSettings=(InterpolationType=Average,RotationOrderForFilter=XZY)
    Child=(Type=Bone,Name="skirt_r")
    Parents=((Item=(Type=Bone,Name="thigh_twist_01_r"),Weight=1))
    Weight=1
   )
   > n1 = &RigUnit_SphericalPoseReader(
    ActiveRegionScaleFactors=(PositiveWidth=1,NegativeWidth=1,PositiveHeight=1,NegativeHeight=1)
    ActiveRegionSize=0.2
    Debug=(bDrawDebug=True,DebugScale=25,DebugSegments=20,DebugThickness=0.25)
    DriverAxis=(X=1,Y=0,Z=0)
    DriverItem=(Type=Bone,Name="thigh_r")
    FalloffRegionScaleFactors=(PositiveWidth=1.5,NegativeWidth=1.5,PositiveHeight=3,NegativeHeight=3)
    FalloffSize=0.4
    OptionalParentItem=(Type=Bone,Name="pelvis")
    RotationOffset=(X=-60,Y=-13,Z=0)
   )
   > &RigUnit_ModifyTransforms(
    ItemToModify="((Item=(Type=Bone,Name=\"bag_lower_MSTR\"),Transform=(Rotation=(X=0,Y=-0.1736,Z=0,W=0.9848),Translation=(X=-2,Y=0,Z=4),Scale3D=(X=1,Y=1,Z=1))),(Item=(Type=Bone,Name=\"skirt_front_r\"),Transform=(Rotation=(X=0.0499,Y=-0.1493,Z=0.0485,W=0.9863),Translation=(X=0, ...(truncated)"
    WeightMinimum=0
    Weight=@n1.OutputParam
   )
   > Set Transforms for Bone_2_3 = RigVMFunctionReferenceNode(
    Joint=(Type=Bone,Name="skirt_front_r")
    Rotate X=-5
    Rotate Y=5
    Rotate Z=-13
    Scale X=0
    Scale Y=0
    Scale Z=0
    Translate X=0
    Translate Y=-3
    Translate Z=0
    Weight=@n1.OutputParam
   )
   > Set Transforms for Bone_2_1_1 = RigVMFunctionReferenceNode(
    Joint=(Type=Bone,Name="skirt_r")
    Rotate X=0
    Rotate Y=-35
    Rotate Z=0
    Scale X=0
    Scale Y=0
    Scale Z=0
    Translate X=0
    Translate Y=0
    Translate Z=5
    Weight=@n1.OutputParam
   )
   > &RigUnit_SphericalPoseReader(
    ActiveRegionScaleFactors=(PositiveWidth=1,NegativeWidth=1,PositiveHeight=1,NegativeHeight=1)
    ActiveRegionSize=0.2
    Debug=(bDrawDebug=True,DebugScale=25,DebugSegments=20,DebugThickness=0.25)
    DriverAxis=(X=-1,Y=0,Z=0)
    DriverItem=(Type=Bone,Name="thigh_l")
    FalloffRegionScaleFactors=(PositiveWidth=1.5,NegativeWidth=1.5,PositiveHeight=3,NegativeHeight=3)
    FalloffSize=0.4
    OptionalParentItem=(Type=Bone,Name="pelvis")
    RotationOffset=(X=110,Y=-13,Z=0)
   )
   > Set Transforms for Bone_2_2_2 = RigVMFunctionReferenceNode(
    Joint=(Type=Bone,Name="skirt_front_l")
    Rotate X=-5
    Rotate Y=4
    Rotate Z=-13
    Scale X=0
    Scale Y=0
    Scale Z=0
    Translate X=0
    Translate Y=2
    Translate Z=0
    Weight=@n1.OutputParam
   )
   > Set Transforms for Bone_2_2_1_1 = RigVMFunctionReferenceNode(
    Joint=(Type=Bone,Name="skirt_l")
    Rotate X=0
    Rotate Y=45
    Rotate Z=0
    Scale X=0
    Scale Y=0
    Scale Z=0
    Translate X=0
    Translate Y=0
    Translate Z=-5
    Weight=@n1.OutputParam
   )
   &RigUnit_ParentConstraint(
    AdvancedSettings=(InterpolationType=Average,RotationOrderForFilter=XZY)
    Child=(Type=Bone,Name="dyn_pigtail_hood_l_base")
    Parents=((Item=(Type=Bone,Name="head"),Weight=0.5),(Item=(Type=Bone,Name="clavicle_l"),Weight=0.5),(Item=(Type=Bone,Name="spine_05"),Weight=0.5))
    Weight=1
    ExecutePin=@RigVMFunction_Sequence.B
   )
   > &RigUnit_AimConstraintLocalSpaceOffset(
    Child=(Type=Bone,Name="dyn_pigtail_hood_l_base")
    Filter=(bX=True,bY=True,bZ=True)
    Parents=((Item=(Type=Bone,Name="spine_05"),Weight=1))
    Weight=1
    WorldUp=(Target=(X=0,Y=0,Z=1),Kind=Location,Space=(Type=Bone,Name="spine_05"))
   )
   > &RigUnit_ModifyTransforms(
    ItemToModify="((Item=(Type=Bone,Name=\"dyn_pigtail_l_base\"),Transform=(Rotation=(X=0,Y=0,Z=0,W=1),Translation=(X=0,Y=0,Z=0),Scale3D=(X=1,Y=1,Z=1))),(Item=(Type=Bone,Name=\"dyn_hood_l_base\"),Transform=(Rotation=(X=0,Y=0,Z=0,W=1),Translation=(X=-2,Y=-1,Z=-1),Scale3D=(X=1,Y= ...(truncated)"
    Weight=1
    WeightMinimum=0
   )
   > &RigUnit_ParentConstraint(
    AdvancedSettings=(InterpolationType=Average,RotationOrderForFilter=XZY)
    Child=(Type=Bone,Name="dyn_pigtail_hood_r_base")
    Parents=((Item=(Type=Bone,Name="head"),Weight=0.5),(Item=(Type=Bone,Name="clavicle_r"),Weight=0.5),(Item=(Type=Bone,Name="spine_05"),Weight=0.5))
    Weight=1
   )
   > &RigUnit_AimConstraintLocalSpaceOffset(
    Child=(Type=Bone,Name="dyn_pigtail_hood_r_base")
    Filter=(bX=True,bY=True,bZ=True)
    Parents=((Item=(Type=Bone,Name="spine_05"),Weight=1))
    Weight=1
    WorldUp=(Target=(X=0,Y=0,Z=1),Kind=Location,Space=(Type=Bone,Name="spine_05"))
   )
   > &RigUnit_ModifyTransforms(
    ItemToModify="((Item=(Type=Bone,Name=\"dyn_pigtail_r_base\"),Transform=(Rotation=(X=0,Y=0,Z=0,W=1),Translation=(X=0,Y=0,Z=0),Scale3D=(X=1,Y=1,Z=1))),(Item=(Type=Bone,Name=\"dyn_hood_r_base\"),Transform=(Rotation=(X=0,Y=0,Z=0,W=1),Translation=(X=-2,Y=-1,Z=-1),Scale3D=(X=1,Y= ...(truncated)"
    Weight=1
    WeightMinimum=0
   )
   &RigUnit_ParentConstraint(
    AdvancedSettings=(RotationOrderForFilter=XZY)
    Child=(Type=Bone,Name="hood_r")
    Parents="((Item=(Type=Bone,Name=\"bow_front\"),Weight=0.0014),(Item=(Type=Bone,Name=\"clavicle_l\"),Weight=0.0014),(Item=(Type=Bone,Name=\"spine_04\"),Weight=0.0035),(Item=(Type=Bone,Name=\"dyn_hood_r\"),Weight=0.0418),(Item=(Type=Bone,Name=\"neck_01\"),Weight=0.0458),(Item= ...(truncated)"
    Weight=1
    ExecutePin=@RigVMFunction_Sequence.C
   )
   > &RigUnit_ParentConstraint(
    AdvancedSettings=(RotationOrderForFilter=XZY)
    Child=(Type=Bone,Name="hood_l")
    Parents="((Item=(Type=Bone,Name=\"bow_front\"),Weight=0.0002),(Item=(Type=Bone,Name=\"spine_04\"),Weight=0.0003),(Item=(Type=Bone,Name=\"dyn_hood_l\"),Weight=0.0004),(Item=(Type=Bone,Name=\"clavicle_r\"),Weight=0.0015),(Item=(Type=Bone,Name=\"neck_01\"),Weight=0.0767),(Item= ...(truncated)"
    Weight=1
   )
   > &RigUnit_ParentConstraint(
    AdvancedSettings=(RotationOrderForFilter=XZY)
    Child=(Type=Bone,Name="hood_back")
    Parents=((Item=(Type=Bone,Name="neck_02"),Weight=0.0001),(Item=(Type=Bone,Name="head"),Weight=0.9999))
    Weight=1
   )
   n2 = &RigUnit_SphericalPoseReader(
    ActiveRegionScaleFactors=(PositiveWidth=1,NegativeWidth=1,PositiveHeight=1,NegativeHeight=1)
    ActiveRegionSize=0.1
    Debug=(bDrawDebug=True,bDraw2D=False,bDrawLocalAxes=False,DebugScale=25,DebugSegments=20,DebugThickness=0.25)
    DriverAxis=(X=0,Y=1,Z=0)
    DriverItem=(Type=Bone,Name="head")
    FalloffRegionScaleFactors=(PositiveWidth=1,NegativeWidth=1,PositiveHeight=1,NegativeHeight=1)
    FalloffSize=0.4
    FlipHeightScaling=False
    FlipWidthScaling=False
    OptionalParentItem=(Type=Bone,Name="spine_05")
    RotationOffset=(X=180,Y=0,Z=0)
    ExecutePin=@RigVMFunction_Sequence.D
   )
   > &RigUnit_ModifyTransforms(
    ItemToModify=((Item=(Type=Bone,Name="dyn_hood_l_base"),Transform=(Rotation=(X=0.3827,Y=0,Z=0,W=0.9239),Translation=(X=-4,Y=0,Z=0),Scale3D=(X=1,Y=1,Z=1))))
    WeightMinimum=0
    Weight=@n2.OutputParam
   )
   > n3 = &RigUnit_SphericalPoseReader(
    ActiveRegionScaleFactors=(PositiveWidth=1,NegativeWidth=1,PositiveHeight=1,NegativeHeight=1)
    ActiveRegionSize=0.1
    Debug=(bDrawDebug=True,bDraw2D=False,bDrawLocalAxes=False,DebugScale=25,DebugSegments=20,DebugThickness=0.25)
    DriverAxis=(X=0,Y=1,Z=0)
    DriverItem=(Type=Bone,Name="head")
    FalloffRegionScaleFactors=(PositiveWidth=1,NegativeWidth=1,PositiveHeight=1,NegativeHeight=1)
    FalloffSize=0.4
    FlipHeightScaling=False
    FlipWidthScaling=False
    OptionalParentItem=(Type=Bone,Name="spine_05")
    RotationOffset=(X=0,Y=0,Z=0)
   )
   > &RigUnit_ModifyTransforms(
    ItemToModify=((Item=(Type=Bone,Name="dyn_hood_r_base"),Transform=(Rotation=(X=0.3827,Y=0,Z=0,W=0.9239),Translation=(X=4,Y=0,Z=0),Scale3D=(X=1,Y=1,Z=1))))
    WeightMinimum=0
    Weight=@n3.OutputParam
   )
   RigVMFunction_ControlFlowBranch = &RigVMFunction_ControlFlowBranch(ExecuteContext=@RigVMFunction_Sequence.E, Condition=RigVMVariableNode() {variable=ClothEnabled}.Value)
   &RigUnit_ModifyTransforms(
    ItemToModify="((Item=(Type=Bone,Name=\"skirt_l\"),Transform=(Rotation=(X=0,Y=0,Z=0,W=1),Translation=(X=0,Y=0,Z=3),Scale3D=(X=1,Y=1,Z=1))),(Item=(Type=Bone,Name=\"skirt_r\"),Transform=(Rotation=(X=0,Y=0,Z=0,W=1),Translation=(X=0,Y=0,Z=-3),Scale3D=(X=1,Y=1,Z=1))),(Item=(Type= ...(truncated)"
    Weight=1
    WeightMinimum=0
    ExecutePin=@RigVMFunction_ControlFlowBranch.False
   )
   &RigUnit_ModifyTransforms(
    ItemToModify=((Item=(Type=Bone,Name="pigtail_01_l"),Transform=(Rotation=(X=0,Y=0.061,Z=0,W=0.9981),Translation=(X=0,Y=0,Z=0),Scale3D=(X=1,Y=1,Z=1))))
    Weight=1
    WeightMinimum=0
    ExecutePin=@RigVMFunction_Sequence.F
   )
   > &RigUnit_ModifyTransforms(
    ItemToModify=((Item=(Type=Bone,Name="pigtail_02_l"),Transform=(Rotation=(X=0,Y=0.1305,Z=0,W=0.9914),Translation=(X=0,Y=0,Z=0),Scale3D=(X=1,Y=1,Z=1))))
    Weight=1
    WeightMinimum=0
   )
   > &RigUnit_ModifyTransforms(
    ItemToModify=((Item=(Type=Bone,Name="pigtail_03_l"),Transform=(Rotation=(X=0,Y=0.1736,Z=0,W=0.9848),Translation=(X=0,Y=0,Z=0),Scale3D=(X=1,Y=1,Z=1))))
    Weight=1
    WeightMinimum=0
   )
   > &RigUnit_ModifyTransforms(
    ItemToModify=((Item=(Type=Bone,Name="pigtail_01_r"),Transform=(Rotation=(X=0,Y=0.061,Z=0,W=0.9981),Translation=(X=0,Y=0,Z=0),Scale3D=(X=1,Y=1,Z=1))))
    Weight=1
    WeightMinimum=0
   )
   > &RigUnit_ModifyTransforms(
    ItemToModify=((Item=(Type=Bone,Name="pigtail_02_r"),Transform=(Rotation=(X=0,Y=0.1305,Z=0,W=0.9914),Translation=(X=0,Y=0,Z=0),Scale3D=(X=1,Y=1,Z=1))))
    Weight=1
    WeightMinimum=0
   )
   > &RigUnit_ModifyTransforms(
    ItemToModify=((Item=(Type=Bone,Name="pigtail_03_r"),Transform=(Rotation=(X=0,Y=0.1736,Z=0,W=0.9848),Translation=(X=0,Y=0,Z=0),Scale3D=(X=1,Y=1,Z=1))))
    Weight=1
    WeightMinimum=0
   )
  ]}
]
hierarchy [
 root {loc=<0 0 0>, rot=<0 0 0 1>}
 pelvis {parent=root, loc=<0 -0.5301 98.0571>, rot=<0.0127 0.707 -0.0127 -0.707>}
 spine_01 {parent=pelvis, loc=<6.1208 -0.9328 0>, rot=<0 0 0.034 -0.9994>}
 spine_02 {parent=spine_01, loc=<6.5996 -0.2095 0>, rot=<0 0 0.0238 -0.9997>}
 spine_03 {parent=spine_02, loc=<6.4914 -0.1237 0>, rot=<0 0 0.0149 -0.9999>}
 spine_04 {parent=spine_03, loc=<6.9598 0.0794 0>, rot=<0 0 -0.0132 -0.9999>}
 spine_05 {parent=spine_04, loc=<6.9987 0 0>, rot=<0 0 -0.0015 -1>}
 clavicle_l {parent=spine_05, loc=<5.4831 -0.1053 -3.2946>, rot=<0.0465 -0.7289 -0.009 -0.6829>}
 upperarm_l {parent=clavicle_l, loc=<9.8658 -1.2136 -0.4105>, rot=<0.0208 -0.486 -0.0242 -0.8734>}
 lowerarm_l {parent=upperarm_l, loc=<26.465 -0.8345 0.259>, rot=<0.0036 0.0161 -0.2298 -0.9731>}
 hand_l {parent=lowerarm_l, loc=<22.7355 -0.3203 -0.6105>, rot=<-0.6784 -0.0223 0.0735 -0.7307>}
 index_metacarpal_l {parent=hand_l, loc=<1.6535 0.0183 -2.96>, rot=<0.0158 -0.0124 -0.0173 -0.9996>}
 index_01_l {parent=index_metacarpal_l, loc=<8.3008 0.0357 -0.0524>, rot=<-0.0044 0.0485 0.0924 -0.9945>}
 index_02_l {parent=index_01_l, loc=<3.3587 -0.3004 0.0452>, rot=<-0.002 0.0027 0.3225 -0.9466>}
 index_03_l {parent=index_02_l, loc=<2.5578 -0.123 -0.0252>, rot=<-0.002 0.0067 0.2558 -0.9667>}
 middle_metacarpal_l {parent=hand_l, loc=<2.1856 -0.1725 -0.5984>, rot=<0.006 0.0168 -0.0329 -0.9993>}
 middle_01_l {parent=middle_metacarpal_l, loc=<7.3275 -0.2447 0.0129>, rot=<-0.0208 0.0042 0.0858 -0.9961>}
 middle_02_l {parent=middle_01_l, loc=<4.0123 0.1401 0.1175>, rot=<-0.0218 0.0258 0.3512 -0.9357>}
 middle_03_l {parent=middle_02_l, loc=<2.6077 0.0534 0.1373>, rot=<0.0082 0.0188 0.1708 -0.9851>}
 pinky_metacarpal_l {parent=hand_l, loc=<1.7616 -0.9753 3.3814>, rot=<-0.0397 0.0814 -0.069 -0.9935>}
 pinky_01_l {parent=pinky_metacarpal_l, loc=<7.1709 -0.4831 -0.1094>, rot=<-0.0654 -0.0035 0.2242 -0.9723>}
 pinky_02_l {parent=pinky_01_l, loc=<2.8274 0.0893 -0.1043>, rot=<0.0037 0.0011 0.2854 -0.9584>}
 pinky_03_l {parent=pinky_02_l, loc=<2.2046 0.0307 -0.0078>, rot=<-0.0002 0.0098 0.1445 -0.9895>}
 ring_metacarpal_l {parent=hand_l, loc=<2.2074 -0.3625 1.3638>, rot=<-0.0284 0.0722 -0.0458 -0.9959>}
 ring_01_l {parent=ring_metacarpal_l, loc=<7.0295 -0.6197 -0.1446>, rot=<-0.0489 -0.0252 0.1892 -0.9804>}
 ring_02_l {parent=ring_01_l, loc=<3.6174 0.4647 -0.1027>, rot=<0 0 0.2806 -0.9598>}
 ring_03_l {parent=ring_02_l, loc=<2.4152 0.114 0.0199>, rot=<0 0 0.1189 -0.9929>}
 thumb_01_l {parent=hand_l, loc=<1.3118 -0.3191 -2.459>, rot=<0.6546 -0.4042 0.0664 -0.6354>}
 thumb_02_l {parent=thumb_01_l, loc=<5.9336 -0.0448 0.0401>, rot=<-0.0037 0.0213 0.298 -0.9543>}
 thumb_03_l {parent=thumb_02_l, loc=<2.2223 0.3682 -0.0907>, rot=<0.0118 -0.013 0.0066 -0.9998>}
 lowerarm_twist_01_l {parent=lowerarm_l, loc=<18.316 0 0>, rot=<0 0 0 1>}
 lowerarm_twist_02_l {parent=lowerarm_l, loc=<10.6707 0 0>, rot=<0 0 0 1>}
 arm_bow_l {parent=lowerarm_l, loc=<5.461 0 4.6206>, rot=<0 0 0 1>}
 upperarm_twist_01_l {parent=upperarm_l, loc=<7.2135 -0.1134 0>, rot=<0 0 0 1>}
 arm_bow_upper_l {parent=upperarm_twist_01_l, loc=<0.9346 -1.0276 4.1272>, rot=<-0.0008 0.0165 0.0374 -0.9992>}
 upperarm_twist_02_l {parent=upperarm_l, loc=<19.7998 -0.1134 0>, rot=<0 0 0 1>}
 Charm_Attach_00 {parent=clavicle_l, type=Null, loc=<5.7463 3.5136 3.4271>, rot=<0.2725 0.2351 -0.7065 0.6094>}
 clavicle_r {parent=spine_05, loc=<5.4831 -0.1053 3.2946>, rot=<0.7289 0.0465 -0.6829 0.009>}
 upperarm_r {parent=clavicle_r, loc=<-9.8658 1.2136 0.4104>, rot=<0.0208 -0.486 -0.0242 -0.8734>}
 lowerarm_r {parent=upperarm_r, loc=<-26.4651 0.8345 -0.259>, rot=<0.0036 0.0161 -0.2298 -0.9731>}
 hand_r {parent=lowerarm_r, loc=<-22.7356 0.3203 0.6106>, rot=<-0.6784 -0.0223 0.0735 -0.7307>}
 index_metacarpal_r {parent=hand_r, loc=<-1.6534 -0.0181 2.9601>, rot=<0.0158 -0.0124 -0.0173 -0.9996>}
 index_01_r {parent=index_metacarpal_r, loc=<-8.3008 -0.0357 0.0524>, rot=<-0.0044 0.0485 0.0924 -0.9945>}
 index_02_r {parent=index_01_r, loc=<-3.3589 0.3003 -0.0453>, rot=<-0.002 0.0027 0.3225 -0.9466>}
 index_03_r {parent=index_02_r, loc=<-2.5578 0.1231 0.0251>, rot=<-0.002 0.0067 0.2558 -0.9667>}
 middle_metacarpal_r {parent=hand_r, loc=<-2.1856 0.1726 0.5984>, rot=<0.006 0.0168 -0.0329 -0.9993>}
 middle_01_r {parent=middle_metacarpal_r, loc=<-7.3273 0.2447 -0.0129>, rot=<-0.0208 0.0042 0.0858 -0.9961>}
 middle_02_r {parent=middle_01_r, loc=<-4.0124 -0.14 -0.1175>, rot=<-0.0218 0.0258 0.3512 -0.9357>}
 middle_03_r {parent=middle_02_r, loc=<-2.6077 -0.0533 -0.1373>, rot=<0.0082 0.0188 0.1708 -0.9851>}
 pinky_metacarpal_r {parent=hand_r, loc=<-1.7617 0.9753 -3.3814>, rot=<-0.0397 0.0814 -0.069 -0.9935>}
 pinky_01_r {parent=pinky_metacarpal_r, loc=<-7.1707 0.4832 0.1094>, rot=<-0.0654 -0.0035 0.2242 -0.9723>}
 pinky_02_r {parent=pinky_01_r, loc=<-2.8274 -0.0892 0.1043>, rot=<0.0037 0.0011 0.2854 -0.9584>}
 pinky_03_r {parent=pinky_02_r, loc=<-2.2045 -0.031 0.0078>, rot=<-0.0002 0.0098 0.1445 -0.9895>}
 ring_metacarpal_r {parent=hand_r, loc=<-2.2073 0.3627 -1.3638>, rot=<-0.0284 0.0722 -0.0458 -0.9959>}
 ring_01_r {parent=ring_metacarpal_r, loc=<-7.0297 0.6197 0.1446>, rot=<-0.0489 -0.0252 0.1892 -0.9804>}
 ring_02_r {parent=ring_01_r, loc=<-3.6171 -0.4648 0.1027>, rot=<0 0 0.2806 -0.9598>}
 ring_03_r {parent=ring_02_r, loc=<-2.4153 -0.1141 -0.0199>, rot=<0 0 0.1189 -0.9929>}
 thumb_01_r {parent=hand_r, loc=<-1.3119 0.3192 2.459>, rot=<0.6546 -0.4042 0.0664 -0.6354>}
 thumb_02_r {parent=thumb_01_r, loc=<-5.9335 0.0447 -0.0401>, rot=<-0.0037 0.0213 0.298 -0.9543>}
 thumb_03_r {parent=thumb_02_r, loc=<-2.2223 -0.3682 0.0907>, rot=<0.0118 -0.013 0.0066 -0.9998>}
 lowerarm_twist_01_r {parent=lowerarm_r, loc=<-18.3161 0 0>, rot=<0 0 0 1>}
 lowerarm_twist_02_r {parent=lowerarm_r, loc=<-10.6707 0 0>, rot=<0 0 0 1>}
 arm_bow_r {parent=lowerarm_r, loc=<-5.4649 0 -4.517>, rot=<0 0 0 1>}
 upperarm_twist_01_r {parent=upperarm_r, loc=<-7.2135 0.1134 0>, rot=<0 0 0 1>}
 arm_bandana_const {parent=upperarm_twist_01_r, loc=<-6.7921 4.4041 0.7615>, rot=<-0.1056 -0.1117 -0.7182 -0.6786>}
 arm_bandana {parent=arm_bandana_const, loc=<0 0 0>, rot=<0 0 0 1>}
 upperarm_twist_02_r {parent=upperarm_r, loc=<-19.7998 0.1134 0>, rot=<0 0 0 1>}
 shoulder_cloth {parent=clavicle_r, loc=<-7.9741 0 -1.5934>, rot=<0 -0.9654 0 -0.2608>}
 shoulder_buckle_r {parent=clavicle_r, loc=<-6.913 -6.196 1.6996>, rot=<0.4573 -0.8685 0.114 0.1534>}
 neck_01 {parent=spine_05, loc=<11.3138 0.1529 0>, rot=<0 0 -0.2003 -0.9797>}
 neck_02 {parent=neck_01, loc=<4.5627 0 0>, rot=<0 0 0.0504 -0.9987>}
 head {parent=neck_02, loc=<5.3914 0 0>, rot=<0 0 0.115 -0.9934>}
 hood_back {parent=head, loc=<9.3774 -6.2176 0.0676>, rot=<0 0 0.6318 -0.7751>}
 hood_r {parent=head, loc=<-6.1371 -6.6872 6.859>, rot=<0 0 0.6318 -0.7751>}
 hood_l {parent=head, loc=<-5.9833 -6.6325 -6.6577>, rot=<0 0 0.6318 -0.7751>}
 bangs {parent=head, loc=<8.7719 9.7704 -1.8759>, rot=<-0.0647 -0.9581 0.0211 0.2784>}
 dyn_pigtail_hood_l_base {parent=head, loc=<-6.0272 -1.381 -7.6743>, rot=<0.5023 0.4836 0.5151 0.4984>}
 dyn_pigtail_l_base {parent=dyn_pigtail_hood_l_base, loc=<0 0 0>, rot=<0 0 0 1>}
 pigtail_01_l {parent=dyn_pigtail_l_base, loc=<0 0 0>, rot=<0 0 0 1>}
 pigtail_02_l {parent=pigtail_01_l, loc=<6.6833 0 -3.1305>, rot=<0 -0.3138 0 -0.9495>}
 pigtail_03_l {parent=pigtail_02_l, loc=<8.7666 0 0>, rot=<0 0 0 1>}
 pigtail_04_l {parent=pigtail_03_l, loc=<10.1593 0 0>, rot=<0 0 0 1>}
 dyn_hood_l_base {parent=dyn_pigtail_hood_l_base, loc=<0 0 0>, rot=<0 0 0 1>}
 dyn_hood_l {parent=dyn_hood_l_base, loc=<0 0 0>, rot=<0 0 0 1>}
 dyn_pigtail_hood_r_base {parent=head, loc=<-6.027 -1.381 7.6743>, rot=<-0.4836 0.5023 0.4984 -0.5151>}
 dyn_hood_r_base {parent=dyn_pigtail_hood_r_base, loc=<-0.8 0 0>, rot=<0 0 0 1>}
 dyn_hood_r {parent=dyn_hood_r_base, loc=<0 0 0>, rot=<0 0 0 1>}
 dyn_pigtail_r_base {parent=dyn_pigtail_hood_r_base, loc=<0 0 0>, rot=<0 0 0 1>}
 pigtail_01_r {parent=dyn_pigtail_r_base, loc=<0 0 0>, rot=<0 0 0 1>}
 pigtail_02_r {parent=pigtail_01_r, loc=<-6.6833 0 3.1307>, rot=<0 -0.3138 0 -0.9495>}
 pigtail_03_r {parent=pigtail_02_r, loc=<-8.7665 0 -0.0002>, rot=<0 0 0 1>}
 pigtail_04_r {parent=pigtail_03_r, loc=<-10.1595 0 0.0002>, rot=<0 0 0 1>}
 back_bow {parent=spine_05, loc=<-0.821 -10.6438 -0.043>, rot=<0.0284 -0.7065 -0.0284 -0.7065>}
 back_bow_r {parent=back_bow, loc=<-1.0725 0.2901 -2.2444>, rot=<0 -0.7053 0 -0.7089>}
 back_bow_l {parent=back_bow, loc=<1.1146 0.2901 -2.2331>, rot=<0 -0.7053 0 -0.7089>}
 bow_front {parent=spine_05, loc=<7.5889 8.1826 -0.4436>, rot=<-0.1739 -0.6952 -0.6734 -0.1815>}
 necklace {parent=spine_05, loc=<6.4576 5.2407 0.049>, rot=<-0.2484 -0.662 0.6606 0.2521>}
 chest {parent=spine_04, loc=<1.1289 7.0202 0>, rot=<0 0 -0.0616 -0.9981>}
 thigh_l {parent=pelvis, loc=<-1.0855 -0.2109 -8.7053>, rot=<0.0009 0.0355 0.0155 -0.9993>}
 calf_l {parent=thigh_l, loc=<-42.5284 -0.0019 -0.7297>, rot=<0.0006 -0.0161 -0.0391 -0.9991>}
 calf_twist_01_l {parent=calf_l, loc=<-35.8483 -1.9197 -0.5794>, rot=<0.0237 0.0095 -0.0035 -0.9997>}
 calf_twist_02_l {parent=calf_l, loc=<-19.1546 -1.7943 -0.2638>, rot=<0.0237 0.0095 -0.0035 -0.9997>}
 foot_l {parent=calf_l, loc=<-46.7402 -2.5014 -0.592>, rot=<0.0427 -0.0238 0.0353 -0.9982>}
 ball_l {parent=foot_l, loc=<-3.5658 10.8066 -0.1544>, rot=<0.0001 -0.0001 -0.7063 -0.7079>}
 thigh_twist_01_l {parent=thigh_l, loc=<-27.952 0 0>, rot=<0.0013 0 0 -1>}
 skirt_front_l {parent=thigh_l, loc=<5.1568 7.3258 1.2471>, rot=<0.0353 0.0033 0.9942 0.102>}
 thigh_r {parent=pelvis, loc=<-1.0853 -0.2109 8.7053>, rot=<0.0355 -0.0009 0.9993 0.0155>}
 calf_r {parent=thigh_r, loc=<42.5286 0.0019 0.7297>, rot=<0.0006 -0.0161 -0.0391 -0.9991>}
 calf_twist_01_r {parent=calf_r, loc=<35.8485 1.9197 0.5794>, rot=<0.0237 0.0095 -0.0035 -0.9997>}
 calf_twist_02_r {parent=calf_r, loc=<19.1546 1.7943 0.2638>, rot=<0.0237 0.0095 -0.0035 -0.9997>}
 foot_r {parent=calf_r, loc=<46.7403 2.5014 0.592>, rot=<0.0427 -0.0238 0.0353 -0.9982>}
 ball_r {parent=foot_r, loc=<3.5658 -10.8066 0.1544>, rot=<0.0001 -0.0001 -0.7063 -0.7079>}
 thigh_twist_01_r {parent=thigh_r, loc=<27.952 0 0>, rot=<0.0013 0 0 -1>}
 skirt_front_r {parent=thigh_r, loc=<-4.9862 -8.1042 -0.7835>, rot=<0.0033 -0.0353 0.102 -0.9942>}
 dagger {parent=pelvis, loc=<6.3517 -0.1199 -14.607>, rot=<0.2089 -0.5208 -0.1853 -0.8067>}
 bag_upper {parent=pelvis, loc=<10.1147 5.192 7.7455>, rot=<0.8535 0.2001 0.4506 0.169>}
 tails01 {parent=pelvis, loc=<7.9327 -2.6786 11.7724>, rot=<-0.0176 -0.9847 -0.0031 0.1733>}
 tails02 {parent=tails01, loc=<16.738 -0.4817 -0.7853>, rot=<0 0 0 -1>}
 tails03 {parent=tails02, loc=<12.724 0 1.5242>, rot=<0 0.174 0 -0.9847>}
 tails06 {parent=tails03, loc=<3.0853 -6.4527 -0.9695>, rot=<0 0 0 -1>}
 tails05 {parent=tails03, loc=<6.1227 -0.2175 0.2529>, rot=<0 0 0 -1>}
 tails04 {parent=tails03, loc=<2.1549 -8.5923 2.6601>, rot=<0 0 0 -1>}
 skirt_back {parent=pelvis, loc=<7.7894 -9.033 0>, rot=<0 0 0.9936 -0.1132>}
 skirt_l {parent=pelvis, loc=<7.0045 0 -11.4144>, rot=<0 -0.9811 0 -0.1937>}
 skirt_r {parent=pelvis, loc=<6.4444 0 10.3574>, rot=<0 -0.9865 0 0.1636>}
 bag_lower_MSTR {parent=pelvis, loc=<4.1174 5.0041 7.8>, rot=<0.5663 0.1108 0.8039 0.1442>}
 bag_lower {parent=bag_lower_MSTR, loc=<0 0 0>, rot=<0 0 0 1>}
 Pose_0 {type=Curve, loc=<0 0 0>, rot=<0 0 0 1>}
 Pose_1 {type=Curve, loc=<0 0 0>, rot=<0 0 0 1>}
 default {type=Curve, loc=<0 0 0>, rot=<0 0 0 1>}
 shoulder_buckle_in_r {type=Curve, loc=<0 0 0>, rot=<0 0 0 1>}
 Backpack_BR {type=Null, loc=<-0.0002 -14.0436 129.957>, rot=<-0.0401 0 0 -0.9992>}
 cape {type=Null, loc=<-0.0003 -12.4093 137.2949>, rot=<0 0 0.0401 -0.9992>}
]
variables [ClothEnabled {type=bool, default=False}]
