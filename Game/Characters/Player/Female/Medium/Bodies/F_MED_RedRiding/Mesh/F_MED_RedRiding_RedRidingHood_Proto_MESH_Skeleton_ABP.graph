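# Anim Blueprint for the Red Riding Hood player body, copying the attached parent's pose and adding secondary motion.
# Control Rig and bone constraints feed RigidBody and AnimDynamics for hair, hood and bows, gated by bRigidBodyEnabled.
# AnimDynamics skirt bones serve as the fallback when bClothEnabled is false.
# Each update runs the cloth controller, then sets Chaos cloth velocity scales during fast yaw or non-skydiving falls.
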AnimBlueprint /Game/Characters/Player/Female/Medium/Bodies/F_MED_RedRiding/Mesh/F_MED_RedRiding_RedRidingHood_Proto_MESH_Skeleton_ABP
target_skeleton=&/F_MED_RedRidingHood_Skeleton
graphs [
 AnimGraph {nodes [
   Output Pose(Result=Use cached pose 'Post Skirt Dynamics'()) {layerGroup=None}
   n1 = Control Rig!(Source=Copy Pose From Mesh() {bCopyCurves=true, bUseAttachedParent=true}, Alpha, ClothEnabled:bool=Get bClothEnabled()) -> (Pose) {
    alphaBoolBlend {blendOption=Linear}
    controlRigAssetReference {blueprintRigClass=&/F_MED_Redriding_CtrlRig.F_MED_Redriding_CtrlRig_C}
   }
   n2 = "Constraint - back_bow (PP)"!(ComponentPose=Local To Component(LocalPose=@n1), Alpha=1, ConstraintWeights_0=0.523, ConstraintWeights_1:float=0.477) -> (Pose) {
    alphaBoolBlend {blendOption=Linear}
    boneToModify {boneName=back_bow}
    constraintSetup [{targetBone {boneName=spine_05}}, {targetBone {boneName=spine_04}}]
    constraintWeights [1,1]
   }
   n3 = "Constraint - bag_lower_MSTR (R)"(ComponentPose=@n2, Alpha=1, ConstraintWeights_0=1) {
    alphaBoolBlend {blendOption=Linear}
    boneToModify {boneName=bag_lower_MSTR}
    constraintSetup [{transformType=Rotation, targetBone {boneName=thigh_r}}]
    constraintWeights [1]
   }
   n4 = "Constraint - arm_bandana_const (PPP)"!(ComponentPose=@n3, Alpha=1, ConstraintWeights_0=0.6115, ConstraintWeights_1:float=0.1678, ConstraintWeights_2:float=0.2207) -> (Pose) {
    alphaBoolBlend {blendOption=Linear}
    boneToModify {boneName=arm_bandana_const}
    constraintSetup [{targetBone {boneName=upperarm_twist_01_r}}, {targetBone {boneName=upperarm_r}}, {targetBone {boneName=upperarm_twist_02_r}}]
    constraintWeights [1,1,1]
   }
   n5 = "Constraint - shoulder_buckle_r (PPPP)"!(
    ComponentPose=@n4
    Alpha=1
    ConstraintWeights_0=0.0003
    ConstraintWeights_1:float=0.663
    ConstraintWeights_2:float=0.2311
    ConstraintWeights_3:float=0.1056
   ) -> (Pose) {
    alphaBoolBlend {blendOption=Linear}
    boneToModify {boneName=shoulder_buckle_r}
    constraintSetup [
     {targetBone {boneName=upperarm_twist_01_r}}
     {targetBone {boneName=clavicle_r}}
     {targetBone {boneName=spine_04}}
     {targetBone {boneName=spine_05}}
    ]
    constraintWeights [1,1,1,1]
   }
   n6 = "Transform (Modify) Bone - Bone: tails01"(ComponentPose=@n5, Alpha=1, Translation="0,0,0", Rotation="-5,0,0", Scale="1,1,1") {
    rotationMode=BMM_Additive
    rotationSpace=BCS_BoneSpace
    alphaBoolBlend {blendOption=Linear}
    boneToModify {boneName=tails01}
    rotation {pitch=-5}
   }
   n7 = "Transform (Modify) Bone - Bone: bag_lower_MSTR"(ComponentPose=@n6, Alpha=1, Translation="0,0,0", Rotation="10,0.000,0.000", Scale="1,1,1") {
    rotationMode=BMM_Additive
    rotationSpace=BCS_BoneSpace
    alphaBoolBlend {blendOption=Linear}
    boneToModify {boneName=bag_lower_MSTR}
    rotation {pitch=10}
   }
   n8 = "Transform (Modify) Bone - Bone: shoulder_buckle_r"!(ComponentPose=@n7, bAlphaBoolEnabled:bool=True, Translation="0,0,0", Rotation="0.000,-62.503,-0.000", Scale="1,1,1") -> (Pose) {
    alphaInputType=Bool
    rotationMode=BMM_Additive
    rotationSpace=BCS_BoneSpace
    alphaBoolBlend {blendInTime=0.2, blendOption=Linear, blendOutTime=0.2}
    boneToModify {boneName=shoulder_buckle_r}
    rotation=<rot 0 -62.503 0>
   }
   Save cached pose 'Post Deformation Rig'(Pose=Component To Local(ComponentPose=@n8)) {cacheName=Post Deformation Rig}
   n9 = RigidBody(ComponentPose=Local To Component(LocalPose=Use cached pose 'Post Deformation Rig'())) {
    bClampLinearTranslationLimitToRefPose=true
    bOverrideWorldGravity=true
    overridePhysicsAsset=&/F_MED_RedRiding_Physics
    simulationSpace=BaseBoneSpace
    alphaBoolBlend {blendOption=Linear}
    baseBoneRef {boneName=pelvis}
    componentAppliedLinearAccClamp=<3000 3000 3000>
    overrideWorldGravity {z=-980}
    simSpaceSettings {dampingAlpha=0, maxAngularAcceleration=30, maxAngularVelocity=300, maxLinearAcceleration=10, maxLinearVelocity=770, worldAlpha=1}
   }
   n10 = RigidBody(ComponentPose=@n9) {
    bClampLinearTranslationLimitToRefPose=true
    bOverrideWorldGravity=true
    overridePhysicsAsset=&/F_MED_RedRiding_dyn_hair
    simulationSpace=BaseBoneSpace
    alphaBoolBlend {blendOption=Linear}
    baseBoneRef {boneName=spine_05}
    componentAppliedLinearAccClamp=<3000 3000 3000>
    overrideWorldGravity {z=-980}
    simSpaceSettings {dampingAlpha=0, maxAngularAcceleration=30, maxAngularVelocity=300, maxLinearAcceleration=100, maxLinearVelocity=770, worldAlpha=1}
   }
   n11 = Anim Dynamics - Bone: hood_back(ComponentPose=@n10, Alpha=1, ExternalForce="0,0,0") {
    angularDampingOverride=1
    angularSpringConstant=20000
    bAngularSpring=true
    bOverrideAngularDamping=true
    bUseGravityOverride=true
    gravityScale=0.35
    simulationSpace=RootRelative
    alphaBoolBlend {blendOption=Linear}
    boundBone {boneName=hood_back}
    gravityOverride {z=-980}
    physicsBodyDefinitions [{
      boundBone {boneName=hood_back}
      boxExtents=<30 30 30>
      constraintSetup {angularLimitsMax=<30 30 30>, angularLimitsMin=<-30 -30 -30>, angularTarget=<1 0 0>}
      localJointOffset=<4 0 0>
     }]
    simSpaceSettings {maxAngularAcceleration=770, maxAngularVelocity=300, simSpaceAngularAlpha=1}
   }
   n12 = Anim Dynamics - Bone: dyn_arm_bow_l_upper_01(ComponentPose=@n11, Alpha=1, ExternalForce="0,0,0") {
    angularDampingOverride=1
    angularSpringConstant=20000
    bAngularSpring=true
    bOverrideAngularDamping=true
    bUseGravityOverride=true
    gravityScale=0.35
    simulationSpace=RootRelative
    alphaBoolBlend {blendOption=Linear}
    boundBone {boneName=dyn_arm_bow_l_upper_01}
    gravityOverride {z=-980}
    physicsBodyDefinitions [{
      boundBone {boneName=dyn_arm_bow_l_upper_01}
      boxExtents=<20 30 30>
      constraintSetup {angularLimitsMax=<15 15 15>, angularLimitsMin=<-15 -15 -15>, angularTarget=<1 0 0>}
      localJointOffset=<4 0 0>
     }]
    simSpaceSettings {maxAngularAcceleration=300, maxAngularVelocity=770, simSpaceAngularAlpha=0.01}
   }
   n13 = Anim Dynamics - Bone: dyn_arm_bow_l_upper_02(ComponentPose=@n12, Alpha=1, ExternalForce="0,0,0") {
    angularDampingOverride=1
    angularSpringConstant=20000
    bAngularSpring=true
    bOverrideAngularDamping=true
    bUseGravityOverride=true
    gravityScale=0.35
    simulationSpace=RootRelative
    alphaBoolBlend {blendOption=Linear}
    boundBone {boneName=dyn_arm_bow_l_upper_02}
    gravityOverride {z=-980}
    physicsBodyDefinitions [{
      boundBone {boneName=dyn_arm_bow_l_upper_02}
      boxExtents=<20 30 30>
      constraintSetup {angularLimitsMax=<15 15 15>, angularLimitsMin=<-15 -15 -15>, angularTarget=<1 0 0>}
      localJointOffset=<4 0 0>
     }]
    simSpaceSettings {maxAngularAcceleration=300, maxAngularVelocity=770, simSpaceAngularAlpha=0.01}
   }
   n14 = Anim Dynamics - Bone: dyn_arm_bow_l_lower_01(ComponentPose=@n13, Alpha=1, ExternalForce="0,0,0") {
    angularDampingOverride=0.995
    angularSpringConstant=10000
    bAngularSpring=true
    bOverrideAngularDamping=true
    bUseGravityOverride=true
    gravityScale=0.35
    simulationSpace=RootRelative
    alphaBoolBlend {blendOption=Linear}
    boundBone {boneName=dyn_arm_bow_l_lower_01}
    gravityOverride {z=-980}
    physicsBodyDefinitions [{
      boundBone {boneName=dyn_arm_bow_l_lower_01}
      boxExtents=<15 20 20>
      constraintSetup {angularLimitsMax=<20 5 20>, angularLimitsMin=<-20 -30 -20>, angularTarget=<1 0 0>}
      localJointOffset=<4 0 0>
     }]
    simSpaceSettings {maxAngularAcceleration=300, maxAngularVelocity=770, simSpaceAngularAlpha=1}
   }
   n15 = Anim Dynamics - Bone: dyn_arm_bow_l_lower_02(ComponentPose=@n14, Alpha=1, ExternalForce="0,0,0") {
    angularDampingOverride=0.995
    angularSpringConstant=10000
    bAngularSpring=true
    bOverrideAngularDamping=true
    bUseGravityOverride=true
    gravityScale=0.35
    simulationSpace=RootRelative
    alphaBoolBlend {blendOption=Linear}
    boundBone {boneName=dyn_arm_bow_l_lower_02}
    gravityOverride {z=-980}
    physicsBodyDefinitions [{
      boundBone {boneName=dyn_arm_bow_l_lower_02}
      boxExtents=<15 20 20>
      constraintSetup {angularLimitsMax=<20 5 20>, angularLimitsMin=<-20 -30 -20>, angularTarget=<1 0 0>}
      localJointOffset=<4 0 0>
     }]
    simSpaceSettings {maxAngularAcceleration=300, maxAngularVelocity=770, simSpaceAngularAlpha=1}
   }
   n16 = Anim Dynamics - Bone: dyn_upperarm_bow_l_upper_01(ComponentPose=@n15, Alpha=1, ExternalForce="0,0,0") {
    angularDampingOverride=1
    angularSpringConstant=20000
    bAngularSpring=true
    bOverrideAngularDamping=true
    bUseGravityOverride=true
    gravityScale=0.35
    simulationSpace=RootRelative
    alphaBoolBlend {blendOption=Linear}
    boundBone {boneName=dyn_upperarm_bow_l_upper_01}
    gravityOverride {z=-980}
    physicsBodyDefinitions [{
      boundBone {boneName=dyn_upperarm_bow_l_upper_01}
      boxExtents=<20 30 30>
      constraintSetup {angularLimitsMax=<15 15 15>, angularLimitsMin=<-15 -15 -15>, angularTarget=<1 0 0>}
      localJointOffset=<4 0 0>
     }]
    simSpaceSettings {maxAngularAcceleration=300, maxAngularVelocity=770, simSpaceAngularAlpha=0.01}
   }
   n17 = Anim Dynamics - Bone: dyn_upperarm_bow_l_upper_02(ComponentPose=@n16, Alpha=1, ExternalForce="0,0,0") {
    angularDampingOverride=1
    angularSpringConstant=20000
    bAngularSpring=true
    bOverrideAngularDamping=true
    bUseGravityOverride=true
    gravityScale=0.35
    simulationSpace=RootRelative
    alphaBoolBlend {blendOption=Linear}
    boundBone {boneName=dyn_upperarm_bow_l_upper_02}
    gravityOverride {z=-980}
    physicsBodyDefinitions [{
      boundBone {boneName=dyn_upperarm_bow_l_upper_02}
      boxExtents=<20 30 30>
      constraintSetup {angularLimitsMax=<15 15 15>, angularLimitsMin=<-15 -15 -15>, angularTarget=<1 0 0>}
      localJointOffset=<4 0 0>
     }]
    simSpaceSettings {maxAngularAcceleration=300, maxAngularVelocity=770, simSpaceAngularAlpha=0.01}
   }
   n18 = Anim Dynamics - Bone: dyn_upperarm_bow_l_lower_01(ComponentPose=@n17, Alpha=1, ExternalForce="0,0,0") {
    angularDampingOverride=0.995
    angularSpringConstant=10000
    bAngularSpring=true
    bOverrideAngularDamping=true
    bUseGravityOverride=true
    gravityScale=0.35
    simulationSpace=RootRelative
    alphaBoolBlend {blendOption=Linear}
    boundBone {boneName=dyn_upperarm_bow_l_lower_01}
    gravityOverride {z=-980}
    physicsBodyDefinitions [{
      boundBone {boneName=dyn_upperarm_bow_l_lower_01}
      boxExtents=<15 20 20>
      constraintSetup {angularLimitsMax=<20 30 20>, angularLimitsMin=<-20 -20 -20>, angularTarget=<1 0 0>}
      localJointOffset=<4 0 0>
     }]
    simSpaceSettings {maxAngularAcceleration=300, maxAngularVelocity=770, simSpaceAngularAlpha=1}
   }
   n19 = Anim Dynamics - Bone: dyn_upperarm_bow_l_lower_02(ComponentPose=@n18, Alpha=1, ExternalForce="0,0,0") {
    angularDampingOverride=0.995
    angularSpringConstant=10000
    bAngularSpring=true
    bOverrideAngularDamping=true
    bUseGravityOverride=true
    gravityScale=0.35
    simulationSpace=RootRelative
    alphaBoolBlend {blendOption=Linear}
    boundBone {boneName=dyn_upperarm_bow_l_lower_02}
    gravityOverride {z=-980}
    physicsBodyDefinitions [{
      boundBone {boneName=dyn_upperarm_bow_l_lower_02}
      boxExtents=<15 20 20>
      constraintSetup {angularLimitsMax=<20 30 20>, angularLimitsMin=<-20 -20 -20>, angularTarget=<1 0 0>}
      localJointOffset=<4 0 0>
     }]
    simSpaceSettings {maxAngularAcceleration=300, maxAngularVelocity=770, simSpaceAngularAlpha=1}
   }
   n20 = Anim Dynamics - Bone: dyn_arm_bow_r_upper_01(ComponentPose=@n19, Alpha=1, ExternalForce="0,0,0") {
    angularDampingOverride=1
    angularSpringConstant=20000
    bAngularSpring=true
    bOverrideAngularDamping=true
    bUseGravityOverride=true
    gravityScale=0.35
    simulationSpace=RootRelative
    alphaBoolBlend {blendOption=Linear}
    boundBone {boneName=dyn_arm_bow_r_upper_01}
    gravityOverride {z=-980}
    physicsBodyDefinitions [{
      boundBone {boneName=dyn_arm_bow_r_upper_01}
      boxExtents=<20 30 30>
      constraintSetup {angularLimitsMax=<15 15 15>, angularLimitsMin=<-15 -15 -15>, angularTarget=<1 0 0>}
      localJointOffset=<-4 0 0>
     }]
    simSpaceSettings {maxAngularAcceleration=300, maxAngularVelocity=770, simSpaceAngularAlpha=0.01}
   }
   n21 = Anim Dynamics - Bone: dyn_arm_bow_r_upper_02(ComponentPose=@n20, Alpha=1, ExternalForce="0,0,0") {
    angularDampingOverride=1
    angularSpringConstant=20000
    bAngularSpring=true
    bOverrideAngularDamping=true
    bUseGravityOverride=true
    gravityScale=0.35
    simulationSpace=RootRelative
    alphaBoolBlend {blendOption=Linear}
    boundBone {boneName=dyn_arm_bow_r_upper_02}
    gravityOverride {z=-980}
    physicsBodyDefinitions [{
      boundBone {boneName=dyn_arm_bow_r_upper_02}
      boxExtents=<20 30 30>
      constraintSetup {angularLimitsMax=<15 15 15>, angularLimitsMin=<-15 -15 -15>, angularTarget=<1 0 0>}
      localJointOffset=<-4 0 0>
     }]
    simSpaceSettings {maxAngularAcceleration=300, maxAngularVelocity=770, simSpaceAngularAlpha=0.01}
   }
   n22 = Anim Dynamics - Bone: dyn_arm_bow_r_lower_01(ComponentPose=@n21, Alpha=1, ExternalForce="0,0,0") {
    angularDampingOverride=0.995
    angularSpringConstant=10000
    bAngularSpring=true
    bOverrideAngularDamping=true
    bUseGravityOverride=true
    gravityScale=0.35
    simulationSpace=RootRelative
    alphaBoolBlend {blendOption=Linear}
    boundBone {boneName=dyn_arm_bow_r_lower_01}
    gravityOverride {z=-980}
    physicsBodyDefinitions [{
      boundBone {boneName=dyn_arm_bow_r_lower_01}
      boxExtents=<15 20 20>
      constraintSetup {angularLimitsMax=<20 5 20>, angularLimitsMin=<-20 -30 -20>, angularTarget=<1 0 0>}
      localJointOffset=<-4 0 0>
     }]
    simSpaceSettings {maxAngularAcceleration=300, maxAngularVelocity=770, simSpaceAngularAlpha=1}
   }
   n23 = Anim Dynamics - Bone: dyn_arm_bow_r_lower_02(ComponentPose=@n22, Alpha=1, ExternalForce="0,0,0") {
    angularDampingOverride=0.995
    angularSpringConstant=10000
    bAngularSpring=true
    bOverrideAngularDamping=true
    bUseGravityOverride=true
    gravityScale=0.35
    simulationSpace=RootRelative
    alphaBoolBlend {blendOption=Linear}
    boundBone {boneName=dyn_arm_bow_r_lower_02}
    gravityOverride {z=-980}
    physicsBodyDefinitions [{
      boundBone {boneName=dyn_arm_bow_r_lower_02}
      boxExtents=<15 20 20>
      constraintSetup {angularLimitsMax=<20 5 20>, angularLimitsMin=<-20 -30 -20>, angularTarget=<1 0 0>}
      localJointOffset=<-4 0 0>
     }]
    simSpaceSettings {maxAngularAcceleration=300, maxAngularVelocity=770, simSpaceAngularAlpha=1}
   }
   n24 = Anim Dynamics - Bone: back_bow_r(ComponentPose=@n23, Alpha=1, ExternalForce="0,0,0") {
    angularDampingOverride=0.995
    angularSpringConstant=10000
    bAngularSpring=true
    bOverrideAngularDamping=true
    bUseGravityOverride=true
    gravityScale=0.35
    simulationSpace=RootRelative
    alphaBoolBlend {blendOption=Linear}
    boundBone {boneName=back_bow_r}
    gravityOverride {z=-980}
    physicsBodyDefinitions [{
      boundBone {boneName=back_bow_r}
      boxExtents=<15 20 20>
      constraintSetup {angularLimitsMax=<30 30 30>, angularLimitsMin=<-30 -30 -30>, angularTarget=<1 0 0>}
      localJointOffset=<-4 0 0>
     }]
    simSpaceSettings {maxAngularAcceleration=300, maxAngularVelocity=770, simSpaceAngularAlpha=1}
   }
   n25 = Anim Dynamics - Bone: back_bow_l(ComponentPose=@n24, Alpha=1, ExternalForce="0,0,0") {
    angularDampingOverride=0.995
    angularSpringConstant=10000
    bAngularSpring=true
    bOverrideAngularDamping=true
    bUseGravityOverride=true
    gravityScale=0.35
    simulationSpace=RootRelative
    alphaBoolBlend {blendOption=Linear}
    boundBone {boneName=back_bow_l}
    gravityOverride {z=-980}
    physicsBodyDefinitions [{
      boundBone {boneName=back_bow_l}
      boxExtents=<15 20 20>
      constraintSetup {angularLimitsMax=<30 30 30>, angularLimitsMin=<-30 -30 -30>, angularTarget=<1 0 0>}
      localJointOffset=<4 0 0>
     }]
    simSpaceSettings {maxAngularAcceleration=300, maxAngularVelocity=770, simSpaceAngularAlpha=1}
   }
   n26 = Anim Dynamics - Bone: back_bow_side_l(ComponentPose=@n25, Alpha=1, ExternalForce="0,0,0") {
    angularDampingOverride=1
    angularSpringConstant=20000
    bAngularSpring=true
    bOverrideAngularDamping=true
    bUseGravityOverride=true
    gravityScale=0.35
    simulationSpace=RootRelative
    alphaBoolBlend {blendOption=Linear}
    boundBone {boneName=back_bow_side_l}
    gravityOverride {z=-980}
    physicsBodyDefinitions [{
      boundBone {boneName=back_bow_side_l}
      boxExtents=<20 30 30>
      constraintSetup {angularLimitsMax=<15 15 15>, angularLimitsMin=<-15 -15 -15>, angularTarget=<1 0 0>}
      localJointOffset=<4 0 0>
     }]
    simSpaceSettings {maxAngularAcceleration=300, maxAngularVelocity=770, simSpaceAngularAlpha=0.01}
   }
   n27 = Anim Dynamics - Bone: back_bow_side_r(ComponentPose=@n26, Alpha=1, ExternalForce="0,0,0") {
    angularDampingOverride=1
    angularSpringConstant=20000
    bAngularSpring=true
    bOverrideAngularDamping=true
    bUseGravityOverride=true
    gravityScale=0.35
    simulationSpace=RootRelative
    alphaBoolBlend {blendOption=Linear}
    boundBone {boneName=back_bow_side_r}
    gravityOverride {z=-980}
    physicsBodyDefinitions [{
      boundBone {boneName=back_bow_side_r}
      boxExtents=<20 30 30>
      constraintSetup {angularLimitsMax=<15 15 15>, angularLimitsMin=<-15 -15 -15>, angularTarget=<1 0 0>}
      localJointOffset=<-4 0 0>
     }]
    simSpaceSettings {maxAngularAcceleration=300, maxAngularVelocity=770, simSpaceAngularAlpha=0.01}
   }
   n28 = Anim Dynamics - Bone: bow_front_l_upper(ComponentPose=@n27, Alpha=1, ExternalForce="0,0,0") {
    angularDampingOverride=1
    angularSpringConstant=20000
    bAngularSpring=true
    bOverrideAngularDamping=true
    bUseGravityOverride=true
    gravityScale=0.35
    simulationSpace=RootRelative
    alphaBoolBlend {blendOption=Linear}
    boundBone {boneName=bow_front_l_upper}
    gravityOverride {z=-980}
    physicsBodyDefinitions [{
      boundBone {boneName=bow_front_l_upper}
      boxExtents=<15 20 20>
      constraintSetup {angularLimitsMax=<30 30 30>, angularLimitsMin=<-30 -30 -30>, angularTarget=<1 0 0>}
      localJointOffset=<4 0 0>
     }]
    simSpaceSettings {maxAngularAcceleration=300, maxAngularVelocity=770, simSpaceAngularAlpha=1}
   }
   n29 = Anim Dynamics - Bone: bow_front_l_lower(ComponentPose=@n28, Alpha=1, ExternalForce="0,0,0") {
    angularDampingOverride=1
    angularSpringConstant=20000
    bAngularSpring=true
    bOverrideAngularDamping=true
    bUseGravityOverride=true
    gravityScale=0.35
    simulationSpace=RootRelative
    alphaBoolBlend {blendOption=Linear}
    boundBone {boneName=bow_front_l_lower}
    gravityOverride {z=-980}
    physicsBodyDefinitions [{
      boundBone {boneName=bow_front_l_lower}
      boxExtents=<15 20 20>
      constraintSetup {angularLimitsMax=<30 30 30>, angularLimitsMin=<-30 -30 -30>, angularTarget=<1 0 0>}
      localJointOffset=<4 0 0>
     }]
    simSpaceSettings {maxAngularAcceleration=300, maxAngularVelocity=770, simSpaceAngularAlpha=1}
   }
   n30 = Anim Dynamics - Bone: bow_front_r_lower(ComponentPose=@n29, Alpha=1, ExternalForce="0,0,0") {
    angularDampingOverride=0.995
    angularSpringConstant=10000
    bAngularSpring=true
    bOverrideAngularDamping=true
    bUseGravityOverride=true
    gravityScale=0.35
    simulationSpace=RootRelative
    alphaBoolBlend {blendOption=Linear}
    boundBone {boneName=bow_front_r_lower}
    gravityOverride {z=-980}
    physicsBodyDefinitions [{
      boundBone {boneName=bow_front_r_lower}
      boxExtents=<20 30 30>
      constraintSetup {angularLimitsMax=<15 15 15>, angularLimitsMin=<-15 -15 -15>, angularTarget=<1 0 0>}
      localJointOffset=<4 0 0>
     }]
    simSpaceSettings {maxAngularAcceleration=300, maxAngularVelocity=770, simSpaceAngularAlpha=1}
   }
   n31 = Anim Dynamics - Bone: bow_front_r_upper(ComponentPose=@n30, Alpha=1, ExternalForce="0,0,0") {
    angularDampingOverride=0.995
    angularSpringConstant=10000
    bAngularSpring=true
    bOverrideAngularDamping=true
    bUseGravityOverride=true
    gravityScale=0.35
    simulationSpace=RootRelative
    alphaBoolBlend {blendOption=Linear}
    boundBone {boneName=bow_front_r_upper}
    gravityOverride {z=-980}
    physicsBodyDefinitions [{
      boundBone {boneName=bow_front_r_upper}
      boxExtents=<20 30 30>
      constraintSetup {angularLimitsMax=<15 15 15>, angularLimitsMin=<-15 -15 -15>, angularTarget=<1 0 0>}
      localJointOffset=<4 0 0>
     }]
    simSpaceSettings {maxAngularAcceleration=300, maxAngularVelocity=770, simSpaceAngularAlpha=1}
   }
   Save cached pose 'Post RBAN'(Pose=Blend Poses by bool(
     BlendPose_0=Component To Local(ComponentPose=@n31)
     BlendPose_1=Use cached pose 'Post Deformation Rig'()
     bActiveValue=Get bRigidBodyEnabled()
    ) {blendType=Linear}) {cacheName=Post RBAN}
   n32 = Anim Dynamics - Bone: skirt_front_r(ComponentPose=Local To Component(LocalPose=Use cached pose 'Post RBAN'()), Alpha=1, ExternalForce="0,0,0") {
    angularDampingOverride=1
    angularSpringConstant=20000
    bAngularSpring=true
    bOverrideAngularDamping=true
    bUseGravityOverride=true
    gravityScale=0.35
    simulationSpace=RootRelative
    alphaBoolBlend {blendOption=Linear}
    boundBone {boneName=skirt_front_r}
    gravityOverride {z=-980}
    physicsBodyDefinitions [{
      boundBone {boneName=skirt_front_r}
      boxExtents=<20 30 30>
      constraintSetup {angularLimitsMax=<0 10 10>, angularLimitsMin=<0 0 0>, angularTarget=<1 0 0>}
      localJointOffset=<4 0 0>
     }]
    simSpaceSettings {maxAngularAcceleration=300, maxAngularVelocity=770, simSpaceAngularAlpha=1}
   }
   n33 = Anim Dynamics - Bone: skirt_front_l(ComponentPose=@n32, Alpha=1, ExternalForce="0,0,0") {
    angularDampingOverride=1
    angularSpringConstant=20000
    bAngularSpring=true
    bOverrideAngularDamping=true
    bUseGravityOverride=true
    gravityScale=0.35
    simulationSpace=RootRelative
    alphaBoolBlend {blendOption=Linear}
    boundBone {boneName=skirt_front_l}
    gravityOverride {z=-980}
    physicsBodyDefinitions [{
      boundBone {boneName=skirt_front_l}
      boxExtents=<20 30 30>
      constraintSetup {angularLimitsMax=<10 10 10>, angularLimitsMin=<0 0 0>, angularTarget=<1 0 0>}
      localJointOffset=<4 0 0>
     }]
    simSpaceSettings {maxAngularAcceleration=300, maxAngularVelocity=770, simSpaceAngularAlpha=1}
   }
   n34 = Anim Dynamics - Bone: skirt_l(ComponentPose=@n33, Alpha=1, ExternalForce="0,0,0") {
    angularDampingOverride=1
    angularSpringConstant=20000
    bAngularSpring=true
    bOverrideAngularDamping=true
    bUseGravityOverride=true
    gravityScale=0.35
    simulationSpace=RootRelative
    alphaBoolBlend {blendOption=Linear}
    boundBone {boneName=skirt_l}
    gravityOverride {z=-980}
    physicsBodyDefinitions [{
      boundBone {boneName=skirt_l}
      boxExtents=<20 30 30>
      constraintSetup {angularLimitsMax=<10 10 10>, angularLimitsMin=<0 0 0>, angularTarget=<1 0 0>}
      localJointOffset=<4 0 0>
     }]
    simSpaceSettings {maxAngularAcceleration=300, maxAngularVelocity=770, simSpaceAngularAlpha=1}
   }
   n35 = Anim Dynamics - Bone: skirt_back(ComponentPose=@n34, Alpha=1, ExternalForce="0,0,0") {
    angularDampingOverride=1
    angularSpringConstant=20000
    bAngularSpring=true
    bOverrideAngularDamping=true
    bUseGravityOverride=true
    gravityScale=0.35
    simulationSpace=RootRelative
    alphaBoolBlend {blendOption=Linear}
    boundBone {boneName=skirt_back}
    gravityOverride {z=-980}
    physicsBodyDefinitions [{
      boundBone {boneName=skirt_back}
      boxExtents=<20 30 30>
      constraintSetup {angularLimitsMax=<5 5 5>, angularLimitsMin=<0 0 0>, angularTarget=<1 0 0>}
      localJointOffset=<4 0 0>
     }]
    simSpaceSettings {maxAngularAcceleration=300, maxAngularVelocity=770, simSpaceAngularAlpha=1}
   }
   n36 = Anim Dynamics - Bone: skirt_r(ComponentPose=@n35, Alpha=1, ExternalForce="0,0,0") {
    angularDampingOverride=1
    angularSpringConstant=20000
    bAngularSpring=true
    bOverrideAngularDamping=true
    bUseGravityOverride=true
    gravityScale=0.35
    simulationSpace=RootRelative
    alphaBoolBlend {blendOption=Linear}
    boundBone {boneName=skirt_r}
    gravityOverride {z=-980}
    physicsBodyDefinitions [{
      boundBone {boneName=skirt_r}
      boxExtents=<20 30 30>
      constraintSetup {angularLimitsMax=<0 10 10>, angularLimitsMin=<0 0 0>, angularTarget=<1 0 0>}
      localJointOffset=<4 0 0>
     }]
    simSpaceSettings {maxAngularAcceleration=300, maxAngularVelocity=770, simSpaceAngularAlpha=1}
   }
   n37 = "Transform (Modify) Bone - Bone: tails01"(ComponentPose=@n36, Alpha=1, Translation="0,0,-2", Rotation="0,0,5", Scale="1,1,1") {
    rotationMode=BMM_Additive
    rotationSpace=BCS_BoneSpace
    translationMode=BMM_Additive
    translationSpace=BCS_BoneSpace
    alphaBoolBlend {blendOption=Linear}
    boneToModify {boneName=tails01}
    rotation {roll=5}
    translation {z=-2}
   }
   Save cached pose 'Post Skirt Dynamics'(Pose=Blend Poses by bool(BlendPose_0=Use cached pose 'Post RBAN'(), BlendPose_1=Component To Local(ComponentPose=@n37), bActiveValue=Get bClothEnabled()) {blendType=Linear}) {cacheName=Post Skirt Dynamics}
  ]}
 EventGraph {nodes [
   n1 = Event BlueprintUpdateAnimation()
   > "Cloth Controller Function (Parameter Asset) (Nativized)"!(
    execute
    CharacterPartAnimInstance:Custom Character Part Anim Instance Object Reference=Self-Reference()
    ClothingAssetName:Name=new_sim_1
    ClothCoreSettings:"Cloth Core Settings Structure (by ref)"=Get Cloth Core Settings() -> (Cloth Core Settings:Cloth Core Settings Structure)
    WindGustLocalSettings:"Wind Gust Settings Structure (by ref)"=Get Wind Gust Local Settings() -> (Wind Gust Local Settings:Wind Gust Settings Structure)
    ClothUpAxis:Vector="1,0,0"
    LinVelScaleFwdBwdClamp:float=0
    LinVelScaleUpDwnClamp:float=0.7
    DeltaTime:float=@n1.DeltaTimeX
    GameTimeInSeconds:float=GetGameTimeInSeconds()
    StateLogicLayerName:Name=none
    Parameters:Fort Character Dynamics Parameters Object Reference=&/F_MED_RedRidingHood_Parameters
   ) -> (
    then
    OutCurrentState:Name=None
    OutClothParameters:Cloth Parameters Structure
    OutClothCoreSettings:Cloth Core Settings Structure
    OutWindGustSettings:Wind Gust Parameters Structure
    OutWindGustLocalSettings:Wind Gust Settings Structure
    OutWindVelocity:Vector
    OutLinearVelVector:Vector
   )
   n2 = BreakRotator(InRot=Get PawnRotationRate() -> (PawnRotationRate:Rotator))
   > n3 = Branch(Condition="float > float"(A="Absolute (Float)"(A=@n2.Yaw), B=2.5))
   > SetClothingInteractorIfNotSet()
   > SetVelocityScale!(
    execute
    self:Chaos Clothing Interactor Object Reference=Get ChaosClothingInteractor() -> (ChaosClothingInteractor:Chaos Clothing Interactor Object Reference)
    LinearVelocityScale:Vector="0.15,0.15,0.15"
    AngularVelocityScale:float=0.1
    FictitiousAngularScale:float=0
   ) -> (then)
   n4 = OR Boolean(A=Get bIsSkyDiving(), B=Get bIsSkyDivingFromBus() -> (bIsSkyDivingFromBus:bool))
   Branch(execute=@n3.else, Condition=AND Boolean(A=Get bIsFalling(), B=NOT Boolean(A=@n4)))
   > SetClothingInteractorIfNotSet()
   > SetVelocityScale(self=Get ChaosClothingInteractor(), LinearVelocityScale="0.2,0.2,0.2", AngularVelocityScale=0.1, FictitiousAngularScale=0)
  ]}
 SetClothingInteractorIfNotSet {nodes [
   SetClothingInteractorIfNotSet() {is=title K2Node_FunctionEntry}
   n1 = IsValid(Object=Get ChaosClothingInteractor())
   > Branch(Condition=AND Boolean(A=Get bClothEnabled(), B=NOT Boolean(A=@n1)))
   n2 = MakeLiteralString(Value=new_sim_1)
   n3 = GetOwningComponent()
   > n4 = Cast To ChaosClothingInteractor!(
    execute
    Object=GetClothingInteractor!(
     self:Clothing Simulation Interactor Object Reference=GetClothingSimulationInteractor!(self:Skeletal Mesh Component Object Reference=@n3, ClothingAssetName:Name=None) -> (ReturnValue:Clothing Simulation Interactor Object Reference)
     ClothingAssetName:Name=String To Name(InString=@n2)
    ) -> (ReturnValue:Clothing Interactor Object Reference)
   ) -> (then, CastFailed, AsChaos Clothing Interactor:Chaos Clothing Interactor Object Reference) {is=title K2Node_DynamicCast}
   > Set ChaosClothingInteractor(ChaosClothingInteractor:Chaos Clothing Interactor Object Reference=@n4) -> (Output_Get:Chaos Clothing Interactor Object Reference)
  ]}
]
variables [
 ChaosClothingInteractor {type=ChaosClothingInteractor, default=null}
 Cloth Core Settings {
  type=ClothCoreSettings
  default {
   collisionThickness=0.8
   dampingCoefficient=0.05
   frictionCoefficient=0.3
   localDampingCoefficient=0
   numberIterations=2
   numberSubsteps=1
   selfCollisionThickness=0
   animDriveDamping=<0 1>
   animDriveStiffness=<0 1>
   areaStiffness=<1 1>
   bendingStiffness=<1 1>
   edgeStiffness=<1 1>
   tetherStiffness=<1 1>
  }
 }
 clothing_sim_interactor {type=ClothingSimulationInteractorNv, default=null}
 current_actor_z {type=number}
 current_dive_jiggle {type=Rotator}
 InComponentSpace {type=boolean}
 InComponentSpaceLastFrame {type=boolean}
 InterpSpeed {type=number}
 Is_Emoting {type=number}
 LockedOnSimSpace {type=boolean}
 MeshToCopy {type=SkeletalMeshComponent}
 sim_space_rotation {type=Rotator}
 Wind Gust Local Settings {type=WindGustSettings, default {elapsedTime=0, finalGustStrengthLocal=0, gustDuration=0, gustStrength=0, timeBetweenGusts=0}}
]
