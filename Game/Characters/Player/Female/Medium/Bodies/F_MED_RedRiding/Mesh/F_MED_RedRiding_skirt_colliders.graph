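# Physics asset of simplified skirt colliders for the medium female RedRiding player body.
# Single capsules cover the pelvis, both calves, the dagger and two tail bones, while both thigh bodies have no shapes.
# Constraints chain calves through thighs to the pelvis, with the dagger and tail chain also rooted there.
# The dagger and tail joints have free swing and twist, unlike the leg joints.
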PhysicsAsset /Game/Characters/Player/Female/Medium/Bodies/F_MED_RedRiding/Mesh/F_MED_RedRiding_skirt_colliders
bodies {
 calf_l [{shapeType=Capsule, length=34.6826, radius=9.482, center=<-19.9486 -2.8284 0.1383>, rotation=<rot 0 -91.1589 -92.0231>}]
 calf_r [{shapeType=Capsule, length=34.6826, radius=9.482, center=<19.998 2.8274 0.0937>, rotation=<rot 0 88.8411 -87.0231>}]
 dagger [{
   shapeType=Capsule
   length=18.5903
   radius=4.4024
   shapeName=dagger_capsule
   center=<3.8212 0.5506 -6.4429>
   rotation=<rot -19.5028 -90.0001 0.6884>
  }]
 pelvis [{shapeType=Capsule, length=7.4292, radius=10.9492, center=<-3.9848 -1.8644 0.0132>, rotation=<rot -4.1601 -89.9995 -1.4037>}]
 tails01 [{
   shapeType=Capsule
   length=6.8201
   radius=7.618
   shapeName=tails01_capsule
   center=<14.5419 -1.5675 -8.3683>
   rotation=<rot -3.0245 -5.0531 -90.3723>
  }]
 tails02 [{shapeType=Capsule, length=3.5312, radius=8.375, shapeName=tails02_capsule, center=<11.6569 -1.7852 -7.8461>, rotation=<rot -77.6141 -90 0>}]
 thigh_l []
 thigh_r []
}
constraints [
 {bone1Name=calf_l, bone2Name=thigh_l, swing1LimitDegrees=45, swing2LimitDegrees=45, twistLimitDegrees=45}
 {bone1Name=thigh_r, bone2Name=pelvis, swing1LimitDegrees=45, swing2LimitDegrees=45, twistLimitDegrees=45}
 {bone1Name=calf_r, bone2Name=thigh_r, swing1LimitDegrees=45, swing2LimitDegrees=45, twistLimitDegrees=45}
 {bone1Name=thigh_l, bone2Name=pelvis, swing1LimitDegrees=45, swing2LimitDegrees=45, twistLimitDegrees=45}
 {
  bone1Name=tails01
  bone2Name=pelvis
  swing1LimitDegrees=45
  swing1Motion=Free
  swing2LimitDegrees=45
  swing2Motion=Free
  twistLimitDegrees=45
  twistMotion=Free
 }
 {
  bone1Name=tails02
  bone2Name=tails01
  swing1LimitDegrees=45
  swing1Motion=Free
  swing2LimitDegrees=45
  swing2Motion=Free
  twistLimitDegrees=45
  twistMotion=Free
 }
 {
  bone1Name=dagger
  bone2Name=pelvis
  swing1LimitDegrees=45
  swing1Motion=Free
  swing2LimitDegrees=45
  swing2Motion=Free
  twistLimitDegrees=45
  twistMotion=Free
 }
]
